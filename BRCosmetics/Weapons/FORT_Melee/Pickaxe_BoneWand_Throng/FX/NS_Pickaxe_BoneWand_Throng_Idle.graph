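# Idle glow effect for the BoneWand Throng pickaxe, made of one sprite emitter that bursts seven persistent glows.
# Spawn group index selects each glow's size, socket-relative offset and tint, with five orange and two paler, pushed slightly by camera offset.
# Each update follows the socket and scales alpha by a fade-in on system age and a waveform flicker.
# Front-end and time-of-day modifiers adjust color and brightness.
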NiagaraSystem /BRCosmetics/Weapons/FORT_Melee/Pickaxe_BoneWand_Throng/FX/NS_Pickaxe_BoneWand_Throng_Idle
dependencies {
 usedDataInterfaces [
  &NiagaraDataInterfaceArrayFloat
  &NiagaraDataInterfaceArrayFloat2
  &NiagaraDataInterfaceArrayFloat3
  &NiagaraDataInterfaceColorCurve
  &NiagaraDataInterfaceCurve
  &NiagaraDataInterfaceSocketReader
 ]
 usedDynamicInputs [
  &SelectVector2DFromArray
  &MakeCustomFloatFromBool
  &Lerp_Float
 ]
 usedModules [
  &self:SystemSpawnScript.SystemScriptSource.SystemScriptGraph.NiagaraNodeAssignment_2.SetVariables_#1
  &SystemState
 ]
}
emitters {Glow {
  emitterUpdateScript [
   &EmitterState(
    Inactive Response=::NewEnumerator1
    Life Cycle Mode=::NewEnumerator1
    Loop Behavior=::NewEnumerator1
    Loop Duration Mode=::NewEnumerator1
    UseLoopDelay=false
   )
   &SpawnBurst_Instantaneous(Spawn Count=7, Use Spawn Probability=false, UseLoopCountLimit=false)
  ]
  particleSpawnScript [
   &FN_SetSpawnGroup(Number of Spawn Groups=7)
   &InitializeParticle(
    Color=&LerpLinearColors(
     StartColor=&LerpLinearColors(
      StartColor=<rgba 2 0.3387 0 1>
      EndColor=<rgba 2 0.555 0.2604 1>
      LerpFactor=&FN_FrontEndFloatModifier(Gameplay=0, Front End=1)
     )
     EndColor=<rgba 1.5 1.0717 0.9844 1>
     LerpFactor=&SelectFloatFromArray(
      Direct Array Index=&SpawnGroup()
      Float Selection Array=DataInterface() {propertyValues {FloatData [0,0,0,0,0,1,1]}}
     )
    )
    Sprite Size Mode=::NewEnumerator3
    Uniform Sprite Size=&SelectFloatFromArray(
     Direct Array Index=&SpawnGroup()
     Float Selection Array=DataInterface() {propertyValues {FloatData [25,18,25,20,15,11,11]}}
    )
    UsePositionOffset=false
    Write Material Random=true
   )
   &CameraOffset(Camera Offset Amount=4)
  ]
  particleUpdateScript [
   &ParticleState(Kill Particles When Lifetime Has Elapsed=false, Let Infinitely Lived Particles Die When Emitter Deactivates=true)
   &SocketLocation(
    SamplePosOffset=true
    Sampled Position Offset=&SelectVectorFromArray(
     Direct Array Index=&SpawnGroup()
     Vector Selection Array=DataInterface() {propertyValues {FloatData [<-1.44 0 46.508>, <-19.7969 0 33.147>, <-15.8238 0 12.4024>, <17.1956 0 33>, <13.6643 0 13.1457>, <-7 -1 31.7>, <4.963 -1 31.7>]}}
    )
    Socket Reader=$System.Socket Reader Interface
    Socket Sampling Mode=::NewEnumerator2
    VelocitySpeedLimit=false
   )
   &ScaleColor(
    Scale Alpha=&Multiply_Float(
     A=&Multiply_Float(
      A=&FloatFromCurve(FloatCurve=DataInterface() {propertyValues {Curve {keys [<at 0.15 0 RCIM_Cubic>,<at 0.3 1>]}}}, CurveIndex=$System.Age, Scale Curve=0.3)
      B=&Waveform("Waveform [1]"=::NewEnumerator12, "[1] Frequency"=7, "[1] Amplitude Min/Max"=<0.75 1>)
     )
     B=&Multiply_Float(
      A=&FN_FrontEndFloatModifier(Gameplay=0.25, Front End=0.1)
      B=&FNTimeOfDayScalar(Night Scale=0.63, Morning Scale=0.85, Evening Scale=0.8)
     )
    )
    Scale RGB=&VectorFromFloat(Value=&FNTimeOfDayScalar(Night Scale=0.73, Morning Scale=0.86, Evening Scale=0.87))
    ScaleA=true
    ScaleRGB=true
   )
  ]
  rendererClasses [&NiagaraSpriteRendererProperties]
  renderers [{rendererClass=&NiagaraSpriteRendererProperties}]
 }}
summary {
 systemName=NS_Pickaxe_BoneWand_Throng_Idle
 emitters [{emitterName=Glow, rendererClasses [&NiagaraSpriteRendererProperties]}]
 userVariables [
  User.bIsFrontEnd {type=NiagaraBool}
  User.bIsFrontEndPreview {type=NiagaraBool}
 ]
}
