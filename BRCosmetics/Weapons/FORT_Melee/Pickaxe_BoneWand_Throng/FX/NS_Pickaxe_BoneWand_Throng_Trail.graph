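# Niagara ribbon swing trail for the BoneWand Throng pickaxe, layering fiery orange-to-dark-red trails with a dark red screen layer.
# Three short-lived emitters use AnimTrail modules to spawn and update ribbons along bones or sockets, with the fire layers sampling User.SkeletalMesh.
# Lifetime curves drive ribbon width and color: fire layers taper and fade from HDR, while TrailBlackScreen widens and fades.
# Dynamic material parameters add random variation and age-based change.
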NiagaraSystem /BRCosmetics/Weapons/FORT_Melee/Pickaxe_BoneWand_Throng/FX/NS_Pickaxe_BoneWand_Throng_Trail
dependencies {
 usedDataInterfaces [&NiagaraDataInterfaceCurve, &NiagaraDataInterfaceColorCurve]
 usedDynamicInputs []
 usedModules [&SystemState]
}
emitters {
 TrailBlackScreen {
  emitterUpdateScript [
   &EmitterState(Life Cycle Mode=::NewEnumerator1, Loop Behavior=::NewEnumerator1, Loop Duration=0.27, UseLoopDelay=true)
   &FN_SpawnAnimTrail_v2()
  ]
  particleSpawnScript [
   /Game/Effects/Niagara/Modules/AnimTrail/FN_InitializeAnimTrail/FN_InitializeAnimTrailBoneIndecies(First Bone or Socket Index=1)
   /Game/Effects/Niagara/Modules/AnimTrail/FN_InitializeAnimTrail/FN_InitializeAnimTrail(AnimTrailWidth=0.75, Color=<rgba 0 0 0 3>, Lifetime=0.17)
  ]
  particleUpdateScript [
   &FN_UpdateAnimTrail_v2()
   &ScaleRibbonWidth(Ribbon Width Scale=&FloatFromCurve(FloatCurve=DataInterface() {propertyValues {Curve {keys [
         {arriveTangent=-1.2579, interpMode=RCIM_Cubic, time=0, value=1}
         {arriveTangent=0.0144, interpMode=RCIM_Cubic, leaveTangent=0.0144, tangentMode=RCTM_User, time=1, value=1.8}
        ]}}}))
   &DynamicMaterialParameters(
    Index 0 Param 1=&RandomRangeFloat(Minimum=-0.25, Maximum=0.25)
    Index 0 Param 2=1
    Index 0 Param 3=-0.8
    Index 0 Param 4=&FloatFromCurve(FloatCurve=DataInterface() {propertyValues {Curve {keys [
         {interpMode=RCIM_Cubic, tangentMode=RCTM_User, time=0, value=1}
         {arriveTangent=-0.6219, interpMode=RCIM_Cubic, leaveTangent=-0.6219, tangentMode=RCTM_User, time=1, value=0}
        ]}}})
    Module.Param0WriteEnabled=true
    Module.Param1WriteEnabled=true
    Module.Param2WriteEnabled=false
    Module.Param3WriteEnabled=false
   )
   &Color(Color=&ColorFromCurve(ColorCurve=DataInterface() {propertyValues {
       AlphaCurve {keys [<at 0.2272 1>,<at 1 0 RCIM_Cubic>]}
       BlueCurve {keys [<at 0.006 0>,<at 0.2174 0 RCIM_Cubic>,<at 0.5013 0 RCIM_Cubic>,<at 0.8696 0>]}
       GreenCurve {keys [
         <at 0.006 0.0087>
         {arriveTangent=-0.0058, interpMode=RCIM_Cubic, leaveTangent=-0.0058, time=0.2174, value=0.0124}
         {arriveTangent=-0.019, interpMode=RCIM_Cubic, leaveTangent=-0.019, time=0.5013, value=0.0058}
         <at 0.8696 0>
        ]}
       RedCurve {keys [
         <at 0.006 0.1042>
         {arriveTangent=-0.0947, interpMode=RCIM_Cubic, leaveTangent=-0.0947, time=0.2174, value=0.1667}
         {arriveTangent=-0.2555, interpMode=RCIM_Cubic, leaveTangent=-0.2555, time=0.5013, value=0.0573}
         <at 0.8696 0>
        ]}
      }}))
  ]
  rendererClasses [&NiagaraRibbonRendererProperties]
  renderers [{rendererClass=&NiagaraRibbonRendererProperties}]
 }
 TrailFire {
  emitterUpdateScript [
   &EmitterState(Life Cycle Mode=::NewEnumerator1, Loop Behavior=::NewEnumerator1, Loop Duration=0.27, UseLoopDelay=true)
   &FN_SpawnAnimTrail_v2()
  ]
  particleSpawnScript [
   &SampleSkeletalMesh(
    AdditionalYawPitchRoll=false
    Bone / Socket Index:NiagaraInt32=&Add_Integer(A=$Particles.SpawnGroup, B=0)
    Mesh Sampling Type:ENiagaraSkelSamplingModeFull=::NewEnumerator5
    Skeletal Mesh:NiagaraDataInterfaceSkeletalMesh=$User.SkeletalMesh
    Socket Sampling Mode:ENiagaraSocketSamplingMode=::NewEnumerator5
   )
   &FN_InitializeAnimTrailBoneIndecies(First Bone or Socket Index=1)
   &FN_InitializeAnimTrail(AnimTrailWidth=1.2, Color=<rgba 0 0 0 4>, Lifetime=0.17)
  ]
  particleUpdateScript [
   &FN_UpdateAnimTrail_v2()
   &ScaleRibbonWidth(Ribbon Width Scale=&FloatFromCurve(FloatCurve=DataInterface() {propertyValues {Curve {keys [
         {tangentMode=RCTM_User, time=0, value=1}
         {arriveTangent=-0.0405, interpMode=RCIM_Cubic, leaveTangent=-0.0405, tangentMode=RCTM_User, time=0.5, value=1}
         {arriveTangent=-4.7785, interpMode=RCIM_Cubic, leaveTangent=-4.7785, tangentMode=RCTM_User, time=1, value=0}
        ]}}})) {module=ScaleRibbonWidth001}
   &DynamicMaterialParameters(
    Index 0 Param 1=&RandomRangeFloat(Minimum=-0.25, Maximum=0.25)
    Index 0 Param 2=&FloatFromCurve(FloatCurve=DataInterface() {propertyValues {Curve {keys [<at 0 1>,<at 1 0.55>]}}})
    Module.Param0WriteEnabled=true
    Module.Param1WriteEnabled=true
    Module.Param2WriteEnabled=false
    Module.Param3WriteEnabled=false
   )
   &Color(Color=&ColorFromCurve(ColorCurve=DataInterface() {propertyValues {
       AlphaCurve {keys [
         {arriveTangent=-3.2415, interpMode=RCIM_Cubic, leaveTangent=-3.2415, tangentMode=RCTM_User, time=0.0615, value=1.1}
         <at 0.9912 0 RCIM_Cubic>
        ]}
       BlueCurve {keys [
         <at 0.0075 0>
         <at 0.1127 0 RCIM_Cubic>
         {arriveTangent=0.005, interpMode=RCIM_Cubic, leaveTangent=0.005, time=0.2753, value=0}
         <at 0.7937 0.0034>
        ]}
       GreenCurve {keys [
         <at 0.0075 1.1094>
         {arriveTangent=-4.0262, interpMode=RCIM_Cubic, leaveTangent=-4.0262, time=0.1127, value=0.4167}
         {arriveTangent=-0.6118, interpMode=RCIM_Cubic, leaveTangent=-0.6118, time=0.2753, value=0.031}
         <at 0.7937 0>
        ]}
       RedCurve {keys [
         <at 0.0075 6>
         {arriveTangent=-19.4482, interpMode=RCIM_Cubic, leaveTangent=-19.4482, time=0.1127, value=5}
         {arriveTangent=-7.2271, interpMode=RCIM_Cubic, leaveTangent=-7.2271, time=0.2753, value=0.791}
         <at 0.7937 0.0781>
        ]}
      }}))
  ]
  rendererClasses [&NiagaraRibbonRendererProperties]
  renderers [{rendererClass=&NiagaraRibbonRendererProperties}]
 }
 TrailFireScreen {
  emitterUpdateScript [
   &EmitterState(Life Cycle Mode=::NewEnumerator1, Loop Behavior=::NewEnumerator1, Loop Duration=0.27, UseLoopDelay=true)
   &FN_SpawnAnimTrail_v2()
  ]
  particleSpawnScript [
   &SampleSkeletalMesh(
    AdditionalYawPitchRoll=false
    Bone / Socket Index:NiagaraInt32=&Add_Integer(A=$Particles.SpawnGroup, B=0)
    Mesh Sampling Type:ENiagaraSkelSamplingModeFull=::NewEnumerator5
    Skeletal Mesh:NiagaraDataInterfaceSkeletalMesh=$User.SkeletalMesh
    Socket Sampling Mode:ENiagaraSocketSamplingMode=::NewEnumerator5
   )
   &FN_InitializeAnimTrailBoneIndecies(First Bone or Socket Index=1)
   &FN_InitializeAnimTrail(AnimTrailWidth=0.6, Color=<rgba 0 0 0 4>, Lifetime=0.17)
  ]
  particleUpdateScript [
   &FN_UpdateAnimTrail_v2()
   &ScaleRibbonWidth(Ribbon Width Scale=&FloatFromCurve(FloatCurve=DataInterface() {propertyValues {Curve {keys [
         {tangentMode=RCTM_User, time=0, value=1}
         {arriveTangent=-0.0405, interpMode=RCIM_Cubic, leaveTangent=-0.0405, tangentMode=RCTM_User, time=0.5, value=1}
         {arriveTangent=-4.7785, interpMode=RCIM_Cubic, leaveTangent=-4.7785, tangentMode=RCTM_User, time=1, value=0}
        ]}}})) {module=ScaleRibbonWidth001}
   &DynamicMaterialParameters(
    Index 0 Param 1=&RandomRangeFloat(Minimum=-0.25, Maximum=0.25)
    Index 0 Param 2=&FloatFromCurve(FloatCurve=DataInterface() {propertyValues {Curve {keys [<at 0 1>,<at 1 0.55>]}}})
    Module.Param0WriteEnabled=true
    Module.Param1WriteEnabled=true
    Module.Param2WriteEnabled=false
    Module.Param3WriteEnabled=false
   )
   &Color(Color=&ColorFromCurve(ColorCurve=DataInterface() {propertyValues {
       AlphaCurve {keys [
         <at 0 0>
         {arriveTangent=-3.2415, interpMode=RCIM_Cubic, leaveTangent=-3.2415, tangentMode=RCTM_User, time=0.0615, value=1.2}
         <at 0.9912 0 RCIM_Cubic>
        ]}
       BlueCurve {keys [
         <at 0.0075 0>
         <at 0.1127 0 RCIM_Cubic>
         {arriveTangent=0.005, interpMode=RCIM_Cubic, leaveTangent=0.005, time=0.2753, value=0}
         <at 0.7937 0.0034>
        ]}
       GreenCurve {keys [
         <at 0.0075 1.1094>
         {arriveTangent=-4.0262, interpMode=RCIM_Cubic, leaveTangent=-4.0262, time=0.1127, value=0.4167}
         {arriveTangent=-0.6118, interpMode=RCIM_Cubic, leaveTangent=-0.6118, time=0.2753, value=0.031}
         <at 0.7937 0>
        ]}
       RedCurve {keys [
         <at 0.0075 6>
         {arriveTangent=-19.4482, interpMode=RCIM_Cubic, leaveTangent=-19.4482, time=0.1127, value=5}
         {arriveTangent=-7.2271, interpMode=RCIM_Cubic, leaveTangent=-7.2271, time=0.2753, value=0.791}
         <at 0.7937 0.0781>
        ]}
      }}))
  ]
  rendererClasses [&NiagaraRibbonRendererProperties]
  renderers [{rendererClass=&NiagaraRibbonRendererProperties}]
 }
}
summary {
 systemName=NS_Pickaxe_BoneWand_Throng_Trail
 emitters [
  {emitterName=TrailFire, rendererClasses [&NiagaraRibbonRendererProperties]}
  {emitterName=TrailBlackScreen, rendererClasses [&NiagaraRibbonRendererProperties]}
  {emitterName=TrailFireScreen, rendererClasses [&NiagaraRibbonRendererProperties]}
 ]
 userVariables [
  User.ColorSwitcher {type=NiagaraFloat, defaultValue=0, description=""}
  User.RibbonIDOffset {type=NiagaraInt32}
  User.SkeletalMesh {
   type=NiagaraDataInterfaceSkeletalMesh
   defaultValue {
    struct=NiagaraExt_VariableValue_DataInterface
    value {dataInterface=&self:NiagaraDataInterfaceSkeletalMesh_0}
   }
  }
 ]
}
